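; Artemis environment kill volume Blueprint that applies environmental damage to actors entering its custom collision volume.
; On component begin overlap, it skips any actor with gameplay tags matching the TC_IgnoreActorsWithTags variable.
; Other actors receive global environment damage, using a scalable-float damage value and a configurable damage gameplay effect.
; BeginPlay, Tick, actor overlap and the construction script contain no logic.
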(event EventBeginPlay)

(event Collision|EventActorBeginOverlap
  (params (OtherActor :type Actor%20Object%20Reference)))

(event EventTick
  (params (DeltaSeconds :type Float%20%28single-precision%29)))

(event OnComponentBeginOverlap(CustomCollisionVolume)
  (params (OverlappedComponent :type Primitive%20Component%20Object%20Reference) (OtherActor :type Actor%20Object%20Reference) (OtherComp :type Primitive%20Component%20Object%20Reference) (OtherBodyIndex :type Integer) (bFromSweep :type Boolean) (SweepResult :type Hit%20Result%20Structure%20%28by%20ref%29))
  (bind _returnvalue (GameplayTags|HasAnyMatchingGameplayTags(Message) OtherActor (Variables|Default|GetTC_IgnoreActorsWithTags)))
  (if (not _returnvalue)
    (Environment|ApplyGlobalEnvironmentDamagetoActor :DamageTargetActor OtherActor :EnvironmentDamage (ScalableFloat|GetValueatLevel (Variables|Default|GetEnvironment_Damage)) :EnvironmentDamageGE (Variables|Default|GetEnvironment%20Damage%20GE))))

(fn ConstructionScript)
