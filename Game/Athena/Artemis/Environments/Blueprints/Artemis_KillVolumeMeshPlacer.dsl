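; Editor-style helper Blueprint for Artemis environments that spawns a replacement or companion actor at every placed actor of a given class, named for kill volume mesh placement.
; The SpawnActors custom event gathers all actors of the configured source class, then spawns the configured spawn class at each one's transform.
; Each spawned actor's name prints to screen as confirmation.
; The construction script is empty.
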(event Custom|SpawnActors
  (bind _outactors (Actor|GetAllActorsOfClass (Variables|Default|GetActor%20Class)))
  (for _array_element _outactors
    (bind _returnvalue (Game|SpawnActorfromClass (Variables|Default|GetSpawnActorClass) (Transformation|GetActorTransform _array_element) "Undefined" "SelectDefaultAtRuntime" 0 0))
    (Development|PrintString (Utilities|GetDisplayName _returnvalue))))

(fn ConstructionScript)
